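# Translucent UI master material that procedurally draws a list button background with a fill, an angled tab region and a border stroke, using no textures.
# The border mask comes from centered absolute UVs, with separate left/right and top/bottom thresholds softened by SmoothStep.
# A rotated, offset UV through SmoothStep splits fill from tab.
# Stroke color blends over that by mask and stroke alpha, while opacity adds stroke alpha to the blended fill/tab alpha.
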Material /FortUILibrary/Foundation/Buttons/Materials/List/M_Button_List_Master
graph {
 properties {
  bUsedWithStaticMesh=false
  blendMode=BLEND_Translucent
  materialDomain=MD_UI
  pixelDepthOffsetMode=PDOM_Legacy
  shadingModel=MSM_Unlit
  usedShadingModels=Unlit
 }
 nodes [
  n1 = VectorParameter() {parameterName=StrokeRGBA, defaultValue=<rgba 1 0.8838 0 1>}
  n2 = TextureCoordinate()
  n3 = Constant2Vector() {g=1, r=1}
  n4 = Abs(None=Add(A=Multiply(A=@n2, B=@n3)) {constB=-0.5})
  n5 = Max(A=Subtract(A=ComponentMask(None=@n4) {g=true}, B=ScalarParameter() {defaultValue=0.45, group=Item_Border, parameterName=TopBottomStroke, sliderMax=1})) {constB=0}
  n6 = ScalarParameter() {parameterName=SSMin}
  n7 = ScalarParameter() {defaultValue=0.025, parameterName=SSMax}
  n8 = Max(A=Subtract(A=ComponentMask(None=@n4) {r=true}, B=ScalarParameter() {defaultValue=0.45, group=Item_Border, parameterName=LeftRightStroke, sliderMax=1})) {constB=0}
  n9 = Clamp(None=Add(
    A=MaterialFunctionCall(Alpha=@n5, Min=@n6, Max=@n7) {materialFunction=&SmoothStep}.Result
    B=MaterialFunctionCall(Alpha=@n8, Min=@n6, Max=@n7) {materialFunction=&SmoothStep}.Result
   ))
  n10 = VectorParameter() {parameterName=FillRGBA, defaultValue=<rgba 0 0.2915 1 0.5>}
  n11 = VectorParameter() {parameterName=TabRGBA, defaultValue=<rgba 1 0 0.5192 0.75>}
  n12 = Rotator(Coordinate=@n2, Time=ScalarParameter() {defaultValue=-0.02, parameterName=TabRotation}) {speed=6.2832}
  n13 = ScalarParameter() {defaultValue=0.5, parameterName=TabOffset}
  n14 = MaterialFunctionCall(
   Alpha=ComponentMask(None=Subtract(A=@n12, B=@n13)) {r=true}
   Min=Constant()
   Max=ScalarParameter() {defaultValue=0.01, parameterName=SSAngleMax}
  ) {materialFunction=&SmoothStep}
  Add(A=LinearInterpolate(B=@n1.A, Alpha=@n9), B=LinearInterpolate(A=@n10.A, B=@n11.A, Alpha=@n14.Result) {constB=0})
  LinearInterpolate(
   A=LinearInterpolate(A=@n10.RGB, B=@n11.RGB, Alpha=@n14.Result) {constB=0}
   B=LinearInterpolate(B=@n1.RGB, Alpha=@n9) {constB=0}
   Alpha=Multiply(A=@n1.A, B=@n9)
  )
 ]
}
